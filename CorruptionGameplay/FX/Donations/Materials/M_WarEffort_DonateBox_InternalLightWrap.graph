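# Additive unlit Niagara mesh-particle material that fakes a light wrapping onto surfaces for the War Effort donation box FX.
# A sphere mask around the particle position plus LightOffset, sized by dynamic parameter LightSize, is evaluated on the scene position behind the translucency.
# A step limits it to scene surfaces within MeshWrapTolerance of the particle mesh.
# Particle alpha scales that mask, while particle color feeds MF_FX_EmissiveAdjustments.
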Material /CorruptionGameplay/FX/Donations/Materials/M_WarEffort_DonateBox_InternalLightWrap
graph {
 properties {
  bUsedWithNiagaraMeshParticles=true
  blendMode=BLEND_Additive
  pixelDepthOffsetMode=PDOM_Legacy
  shadingModel=MSM_Unlit
  usedShadingModels=Unlit
 }
 nodes [
  n1 = ParticleColor()
  MaterialFunctionCall(Emissive=@n1.RGB) {materialFunction=&MF_FX_EmissiveAdjustments}
  MaterialFunctionCall() {materialFunction=&LocalPosition}
  n2 = MaterialFunctionCall() {materialFunction=&WorldPositionBehindTranslucency}
  n3 = SphereMask(
   A=Add(A=VectorParameter() {parameterName=LightOffset, defaultValue=<rgba 0 0 0 1>}.RGB, B=ParticlePositionWS())
   B=@n2.Result
   Radius=DynamicParameter() {paramNames [LightSize,Param2,Param3,Param4]}.LightSize
  ) {attenuationRadius=0.5, hardnessPercent=0}
  n4 = MaterialFunctionCall(Vector 3=Subtract(A=@n2.Result, B=WorldPosition().XYZ)) {materialFunction=/Engine/Functions/Engine_MaterialFunctions02/Utility/VectorLength}
  n5 = ScalarParameter() {parameterName=MeshWrapTolerance}
  Multiply(A=@n1.A, B=Multiply(A=@n3, B=Step(Y=@n4.V3 Length, X=@n5)))
  Power(Base=@n3)
 ]
}
