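# Additive unlit material for Niagara mesh particles, drawing the fringe of the War Effort donation box lightshaft.
# A UV gradient along V blends particle color toward a desaturated version and ramps brightness between two parameters.
# Two MotionBlur texture samples, one panning sideways, form noise that a parameterized SmoothStep erodes along V, weighted by red vertex color.
# Particle alpha and vertex red scale the result before an emissive adjustment function, and a CameraOffset function provides a depth nudge.
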Material /CorruptionGameplay/FX/Donations/Materials/M_WarEffort_DonationBox_Lightshaft_Fringe
graph {
 properties {
  bUsedWithNiagaraMeshParticles=true
  blendMode=BLEND_Additive
  pixelDepthOffsetMode=PDOM_Legacy
  shadingModel=MSM_Unlit
  usedShadingModels=Unlit
 }
 nodes [
  n1 = ParticleColor()
  n2 = ComponentMask(None=TextureCoordinate() {vTiling=0.5}) {g=true}
  n3 = LinearInterpolate(A=@n1.RGB, B=Desaturation(None=@n1.RGB, Fraction=ScalarParameter() {parameterName=Max_Desat}), Alpha=@n2)
  n4 = LinearInterpolate(A=ScalarParameter() {parameterName=min_brightness}, B=ScalarParameter() {parameterName=max_brightness}, Alpha=@n2)
  n5 = Multiply(A=@n1.A, B=VertexColor().R)
  n6 = TextureCoordinate()
  n7 = Add(A=Multiply(A=ComponentMask(None=@n6) {g=true}, B=VertexColor().R), B=ScalarParameter() {parameterName=Step_Pos})
  n8 = TextureCoordinate() {uTiling=2, vTiling=0.5}
  n9 = TextureCoordinate() {vTiling=0.5}
  n10 = Add(
   A=TextureSample(UVs=Panner(Coordinate=@n8) {speedX=0.1}) {texture=/Game/Effects/Fort_Effects/Textures/Tile/MotionBlur}.G
   B=TextureSample(UVs=Panner(Coordinate=@n9)) {texture=&MotionBlur}.G
  )
  n11 = SmoothStep(
   Min=Add(A=@n7, B=ScalarParameter() {parameterName=Step_W})
   Max=@n7
   Value=Multiply(A=Multiply(A=@n10) {constB=0.5}, B=ScalarParameter() {parameterName=Noise_Prof})
  )
  MaterialFunctionCall(Emissive=Multiply(A=Multiply(A=@n3, B=@n4), B=Multiply(A=@n5, B=@n11))) {materialFunction=&MF_FX_EmissiveAdjustments}
  MaterialFunctionCall(Offset Amount=ScalarParameter() {parameterName=cam_off}) {materialFunction=/Engine/Functions/Engine_MaterialFunctions02/WorldPositionOffset/CameraOffset}
 ]
}
